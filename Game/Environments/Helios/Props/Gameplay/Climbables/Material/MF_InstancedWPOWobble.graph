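# Material function for climbable props that adds a hit-triggered wobble to world position offset on instanced meshes.
# Per-instance custom data supplies hit time, location and wobble direction.
# A global-time sine along that direction fades with horizontal distance from the hit and elapsed time, scaled by parameters, then adds to incoming WPO.
# The "Use WPO Wobble" static switch is not wired to the output, so the wobble always applies.
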MaterialFunction /Game/Environments/Helios/Props/Gameplay/Climbables/Material/MF_InstancedWPOWobble
graph {nodes [
  n1 = FunctionInput() {inputType=FunctionInput_MaterialAttributes}
  n2 = GetMaterialAttributes(None=@n1) {attributeGetTypes [#1]}
  n3 = WorldPosition()
  n4 = Constant3Vector() {constant=<rgba 1 1 0 0>}
  n5 = MaterialFunctionCall(
   X=PerInstanceCustomData!(DefaultValue) {constDefaultValue=1, dataIndex=2, desc=Hit Location X}
   Y=PerInstanceCustomData() {dataIndex=3, desc=Hit Location Y}
   Z=PerInstanceCustomData() {dataIndex=4, desc=Hit Location Z}
  ) {materialFunction=&MakeFloat3}
  n6 = Divide(
   A=Distance(A=Multiply(A=@n3.XYZ, B=@n4), B=Multiply(A=@n4, B=@n5.Result))
   B=ScalarParameter() {defaultValue=400, parameterName=WobbleDistance}
  )
  n7 = Sine(None=Multiply(A=Time(), B=ScalarParameter() {defaultValue=1, parameterName=WobbleSpeed}))
  n8 = MaterialFunctionCall(
   X=PerInstanceCustomData() {constDefaultValue=1, dataIndex=5, desc=Wobble Direction X}
   Y=PerInstanceCustomData() {dataIndex=6, desc=Wobble Direction Y}
   Z=PerInstanceCustomData() {dataIndex=7, desc=Wobble Direction Z}
  ) {materialFunction=&MakeFloat3}
  n9 = Multiply(
   A=OneMinus(None=Saturate(None=@n6))
   B=Multiply(A=Multiply(A=@n7, B=@n8.Result), B=ScalarParameter() {defaultValue=1, parameterName=WobbleStrength})
  )
  n10 = Subtract(A=Time(), B=PerInstanceCustomData() {dataIndex=1, desc=Hit time})
  n11 = ScalarParameter() {defaultValue=1.5, parameterName=WobbleSeconds}
  n12 = OneMinus(None=Saturate(None=Divide(A=@n10, B=@n11)))
  n13 = SetMaterialAttributes!(MaterialAttributes=@n2.MaterialAttributes, WorldPositionOffset=Add(A=Multiply(A=@n9, B=@n12), B=@n2.WorldPositionOffset)) {attributeSetTypes [#1]}
  FunctionOutput(None=@n13)
  StaticSwitchParameter(True=@n13, False=@n1) {defaultValue=true, parameterName=Use WPO Wobble}
 ]}
