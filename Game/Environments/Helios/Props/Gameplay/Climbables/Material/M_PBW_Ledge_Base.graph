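# Base material for climbable ledge props, supporting instanced meshes and Nanite, built on the shared layer function MF_CombineLayers.
# Paint is a tinted, desaturated base color, masked by an atlas cell hashed from quantized instance or object position.
# A per-instance timestamp drives a fading blue emissive flash, followed by wobble, Curie fallback and Creative move-tool functions.
# A triplanar-noise-eroded box mask around HandHoldPosition is declared as WeatheringMask.
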Material /Game/Environments/Helios/Props/Gameplay/Climbables/Material/M_PBW_Ledge_Base
graph {
 properties {
  bUseMaterialAttributes=true
  bUsedWithInstancedStaticMeshes=true
  bUsedWithNanite=true
  pixelDepthOffsetMode=PDOM_Legacy
  usedShadingModels=Default Lit
 }
 nodes [
  n1 = Divide(
   A=Subtract(A=Time(), B=PerInstanceCustomData!(DefaultValue) {constDefaultValue=-1000000})
   B=ScalarParameter() {defaultValue=0.15, parameterName=Duration}
  )
  n2 = SetMaterialAttributes(
   MaterialAttributes=NamedRerouteUsage()
   Emissive Color=LinearInterpolate(A=Constant3Vector() {constant=<rgba 0.3001 0.3808 0.66 1>}, B=Constant3Vector(), Alpha=Saturate(None=@n1))
  ) {attributeSetTypes [#1]}
  MaterialFunctionCall(MA In=MaterialFunctionCall!(
    In=MaterialFunctionCall!(In=@n2) {materialFunction=&/MF_InstancedWPOWobble}.Result
    UseSphericalMappingOverride
   ) {materialFunction=/Game/Packages/Fortress_SharedMaterials/Base_Material_Functions/MF_CurieFallbackEffect}.Result) {materialFunction=&MF_CreativeMoveTool}
  n3 = PreSkinnedPosition()
  n4 = ShaderStageSwitch(PixelShader=VertexInterpolator(VS=@n3).PS, VertexShader=@n3)
  n5 = Subtract(A=@n4, B=VectorParameter() {parameterName=HandHoldPosition, defaultValue=<rgba 0 0 6 1>}.RGB)
  n6 = Constant3Vector()
  n7 = ScalarParameter() {defaultValue=12, parameterName=WearHardness}
  n8 = MaterialFunctionCall(Float3=ActorPositionWS()) {materialFunction=&BreakOutFloat3Components}
  n9 = Divide(
   A=Add(
    A=@n4
    B=MaterialFunctionCall(X=@n8.G, Y=@n8.G, Z=@n8.G) {materialFunction=&MakeFloat3}.Result
   )
   B=VectorParameter() {parameterName=NoiseScale, defaultValue=<rgba 128 128 200 1>}.RGB
  )
  n10 = ComponentMask(None=@n9) {b=true, g=true}
  n11 = NamedRerouteUsage()
  n12 = ComponentMask(None=@n9) {g=true, r=true}
  n13 = NamedRerouteUsage()
  n14 = PreSkinnedNormal()
  n15 = Abs(None=ShaderStageSwitch(PixelShader=VertexInterpolator(VS=@n14).PS, VertexShader=@n14))
  n16 = LinearInterpolate(
   A=LinearInterpolate(A=TextureSample(UVs=@n10, Tex=@n11).RGB, B=TextureSample(UVs=@n12, Tex=@n13).RGB, Alpha=ComponentMask(None=@n15) {b=true})
   B=TextureSample(UVs=ComponentMask(None=@n9) {b=true, r=true}, Tex=NamedRerouteUsage()).RGB
   Alpha=ComponentMask(None=@n15) {g=true}
  )
  n17 = Subtract(
   A=MaterialFunctionCall!(A=@n5, B=@n6, Bounds=VectorParameter() {parameterName=WeatheringBounds, defaultValue=<rgba 48 80 12 1>}.RGB, Edge Falloff=@n7) {materialFunction=/Engine/Functions/Engine_MaterialFunctions01/Gradient/BoxMask-3D}.Result
   B=Multiply(A=ComponentMask(None=@n16) {g=true}, B=ScalarParameter() {defaultValue=1.85, parameterName=NoiseStrength})
  )
  NamedRerouteDeclaration(None=Saturate(None=Ceil(None=@n17))) {name=WeatheringMask, nodeColor=<rgba 0.4062 1 0 1>}
  NamedRerouteDeclaration(None=TextureObject() {samplerType=SAMPLERTYPE_LinearColor, texture=/EpicBaseTextures/Noises/T_Clouds_Noise2_D}) {name=NoiseTexture, nodeColor=<rgba 0 0.0156 1 1>}
  n18 = TextureObjectParameter() {
   desc="SFX TextureGrayscale You can reuse that texture across layers and functions. "
   group=0_SFX Global
   parameterName=SFX_Grayscale
   samplerType=SAMPLERTYPE_LinearGrayscale
   sortPriority=0
   texture=/Game/Global/Base4Layers_2022/Textures/T_SFX_Grayscale
  }
  n19 = TextureObjectParameter() {
   desc="SFX UniversalTextureMask Uses Dot to select channel used for various parts of material.You can reuse that texture across layers and functions. "
   group=0_SFX Global
   parameterName=SFX_Mask
   samplerType=SAMPLERTYPE_Masks
   sortPriority=0
   texture=/Game/Global/Base4Layers_2022/Textures/T_SFX_M
  }
  n20 = StaticSwitchParameter(True=@n18, False=@n19) {desc=Make sure size is less than 512 and lower, group=0_SFX Global, parameterName=UseSFXGrayscale}
  NamedRerouteDeclaration(None=MaterialFunctionCall!(In=@n20) {materialFunction=/Game/Global/Base4Layers_2022/MaterialFunctions/MF_SFXTexture}.SFX) {name=SFX V3, nodeColor=<rgba 0.6406 0 1 1>}
  NamedRerouteDeclaration(None=@n20) {name=SFX Texture Global, nodeColor=<rgba 0.0312 1 0 1>}
  n21 = TextureObjectParameter() {
   group=1_Layer_Nanite
   parameterName="Normals  Nanite"
   samplerType=SAMPLERTYPE_Normal
   sortPriority=0
   texture=/Engine/EngineMaterials/FlatNormal
  }
  n22 = TextureObjectParameter() {
   group=1_Layer_Nanite
   parameterName="SpecularMasks  Nanite"
   samplerType=SAMPLERTYPE_Masks
   sortPriority=0
   texture=&T_Fortnite_Default_S
  }
  n23 = TextureObjectParameter() {group=1_Layer_Nanite, parameterName=Diffuse Nanite, sortPriority=0, texture=/Game/Global/Textures/Default/Blanks/T_MidgreysRGB}
  n24 = StaticBoolParameter() {desc="Enables SFX for Opacity Mask, mast have Masked blending set in MI", group=1_Layer SFX, parameterName=SFX OpacityMask, sortPriority=1}
  n25 = TextureObjectParameter() {group=1_Layer, parameterName=Diffuse, sortPriority=0, texture=&T_MidgreysRGB}
  n26 = VectorParameter() {desc=Use R or G or B to select your SFX Mask Channel, group=0_SFX BP, parameterName=SFX_Tint_Ch, sortPriority=4, defaultValue=<rgba 1 0 0 0>}
  n27 = TextureObjectParameter() {
   group=1_Layer
   parameterName=SpecularMasks
   samplerType=SAMPLERTYPE_Masks
   sortPriority=0
   texture=&T_Fortnite_Default_S
  }
  n28 = TextureObjectParameter() {group=1_Layer, parameterName=Normals, samplerType=SAMPLERTYPE_Normal, sortPriority=0, texture=&FlatNormal}
  n29 = VectorParameter() {
   group=1_Layer SFX
   parameterName="SFX_SSS Tint(Expand)"
   sortPriority=0
   channelNames {a=SSS Depth 0-1, b="", g="", r=""}
   defaultValue=<rgba 0 0 0 0.5>
  }
  n30 = VectorParameter() {
   desc=Use R and G channels of this input to controll ClearCoat amount and Rounghness. It uses SFX texture to mask area affected.
   group=1_Layer SFX
   parameterName="SFX_ClearCoat(Expand)"
   sortPriority=-3
   channelNames {a=Not Used, b=Not Used, g=Clearcoat Roughness, r=Clearcoat Value}
   defaultValue=<rgba 1 0 0 0>
  }
  n31 = ScalarParameter() {desc=If you need to cancel shared Mask texture on specific layer, group=1_Layer SFX, parameterName=SFX_OpacityMaskCancel, sortPriority=3}
  n32 = VectorParameter() {
   desc="Emissive color, by default mult by Base color. You can control that blending via Alpha channel of this input. Uses SFX texture as the Mask"
   group=1_Layer SFX
   parameterName="SFX_Emissive_Tint(Expand)"
   sortPriority=-6
   channelNames {a=Base color to Emissive, b="", g="", r=""}
   defaultValue=<rgba 0 0 0 1>
  }
  n33 = ScalarParameter() {
   bUseCustomPrimitiveData=true
   desc="For now Scalar, this is possible alternative to Texture Swatch. Uses single ramp texture shared between all 4 layers and can have indepent row selected per layer. Enabled globally via function. Will disable use of color swatch if enabled. Diffuse tint still can be used if needed via separate mask "
   group=Y_CPD
   parameterName=CPD_PackHSV_1
   primitiveDataIndex=30
   sliderMin=1
  }
  n34 = GetMaterialAttributes(None=MaterialFunctionCall!(
    WSM Scale
    Use WS Mapping
    Use Z
    Use XYZ
    WSM_ProjectionTransitionContrast
    UV2 True
    WSM_HeightLerpUsingAlpha
    WSM_HeightLerpUsingAlpha_Contrast
    UV's=MaterialFunctionCall!(L3 UV2, L2 UV2, L4 UV2) {materialFunction=/Game/Global/Base4Layers_2022/MaterialFunctions/MF_UVsSet_New}.UVs Layer 1
    Normals_NN=@n21
    SpecularMasks_NN=@n22
    Diffuse_NN=@n23
    Use SFX Emissive=StaticBoolParameter() {desc=Enables SFX, group=1_Layer SFX, parameterName=SFX Emissive, sortPriority=-8}
    SFX Opacity Mask=@n24
    SFX SSS Two Sided Foliange=StaticBoolParameter() {group=1_Layer SFX, parameterName=SFX_SSS Foliage, sortPriority=-2}
    SFX SSS=StaticBoolParameter() {group=1_Layer SFX, parameterName=SFX_SSS, sortPriority=-2}
    SFX Clearcoat=StaticBoolParameter() {group=1_Layer SFX, parameterName=SFX_Clearcoat, sortPriority=-5}
    Diffuse=NamedRerouteDeclaration(None=@n25) {name=L1D, nodeColor=<rgba 0 0.2969 1 1>}
    World Position Z=Constant() {r=1}
    SFX Texture Complete=NamedRerouteUsage()
    SFX Tint Channel=@n26.RGB
    SpecularMasks=NamedRerouteDeclaration(None=@n27) {name=L1S, nodeColor=<rgba 1 0.7734 0 1>}
    Normals=NamedRerouteDeclaration(None=@n28) {name=L1N, nodeColor=<rgba 1 0.3984 0 1>}
    Normals Power=ScalarParameter() {defaultValue=1, group=1_Layer, parameterName=NormalsValue, sortPriority=1}
    "SFX  SSS Channel"=VectorParameter() {group=1_Layer SFX, parameterName=SFX_SSS_Ch, sortPriority=-1}.RGB
    SSS Color Value=AppendVector(A=@n29, B=@n29)
    "Clearcoat (R) CC Roughness (G)"=@n30
    Clearcoat Mask Channal=VectorParameter() {group=1_Layer SFX, parameterName=SFX_Clearcoat_Chl, sortPriority=-4}.RGB
    SFX Opacilty Mask Channel=VectorParameter() {group=1_Layer SFX, parameterName=SFX_OpacityMask_Ch_1, sortPriority=2}.RGB
    Opacilty Mask Cancel=@n31
    "SFX Emissive  Channel"=VectorParameter() {group=1_Layer SFX, parameterName=SFX_Emissive_Ch, sortPriority=-7}.RGB
    Emissive Tint=AppendVector(A=@n32, B=@n32)
    Data Ramp RGB=MaterialFunctionCall!(In=@n33) {materialFunction=/Game/Global/Base4Layers_2022/MaterialFunctions/MF_ColorTint_Unpack}.Result
    Output Diffuse Only=StaticBoolParameter() {group=1_Layer, parameterName=Layer 1 Diffuse Only}
    Output Normals Only
   ) {materialFunction=/Game/Global/Base4Layers_2022/MaterialFunctions/MF_CombineLayers}.Result) {attributeGetTypes [#2]}
  n35 = Multiply(
   A=Desaturation(None=@n34.Base Color, Fraction=Constant() {r=1})
   B=VectorParameter() {parameterName=PaintColor, defaultValue=<rgba 3 3 3 1>}.RGB
  )
  n36 = TextureCoordinate()
  n37 = ScalarParameter() {defaultValue=2, parameterName=Rows}
  n38 = Divide(B=@n37)
  n39 = PerInstanceCustomData3Vector!(DefaultValue) {dataIndex=9}
  n40 = ObjectPositionWS()
  n41 = Floor(None=Divide(A=Max(A=@n39, B=@n40), B=ScalarParameter() {defaultValue=128, parameterName=Randomization_Radius}))
  n42 = MaterialFunctionCall(Float2=Round(None=MaterialFunctionCall(Vec3=@n41) {materialFunction=&MF_Hash}.Vec3)) {materialFunction=&BreakOutFloat2Components}
  n43 = AppendVector(A=LinearInterpolate(Alpha=@n42.R), B=LinearInterpolate(Alpha=@n42.G))
  n44 = TextureSampleParameter2D(UVs=Add(A=Multiply(A=@n36, B=@n38), B=Divide(A=@n43, B=@n37))) {
   parameterName=WeatheringMask_Atlas
   samplerType=SAMPLERTYPE_Masks
   texture=/Game/Environments/Helios/Props/Gameplay/Climbables/Textures/T_Climbing_WoodBeam_A_A_M
  }
  NamedRerouteDeclaration(None=SetMaterialAttributes!(MaterialAttributes=@n34.MaterialAttributes, Base Color=LinearInterpolate(A=@n34.Base Color, B=@n35, Alpha=@n44.R)) {attributeSetTypes [#2]}) {name=L1, nodeColor=<rgba 0 0.5078 1 1>}
  n45 = TextureObjectParameter() {
   desc="SFX RGB - Graffiti, emissive RGB "
   group=0_SFX Global
   parameterName=SFX_RGB
   sortPriority=1
   texture=/Game/Global/Base4Layers_2022/Textures/T-GraffitiSample_D
  }
  NamedRerouteDeclaration(None=@n45) {name=SFX Texture RGB, nodeColor=<rgba 0 0.5312 1 1>}
  ObjectPositionWS()
  SphereMask(A=@n5, B=@n6, Radius=ScalarParameter() {defaultValue=24, parameterName=WearRadius}, Hardness=@n7)
  TransformPosition(None=Constant3Vector()) {transformSourceType=TRANSFORMPOSSOURCE_Instance, transformType=TRANSFORMPOSSOURCE_World}
  VertexNormalWS()
  WorldPosition()
 ]
}
